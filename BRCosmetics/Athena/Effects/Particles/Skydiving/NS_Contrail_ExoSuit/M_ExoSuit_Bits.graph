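# Unlit translucent Niagara sprite material for the bits in the ExoSuit skydiving contrail.
# Particle color RGB feeds an emissive adjustment function, while particle alpha multiplies a mask from the bits texture.
# A Random dynamic parameter picks the mask channel: blue for the top third, red otherwise.
# The nested one-third threshold returns red on every branch, so only two variants occur and green is unused.
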Material /BRCosmetics/Athena/Effects/Particles/Skydiving/NS_Contrail_ExoSuit/M_ExoSuit_Bits
graph {
 properties {bUsedWithNiagaraSprites=true, blendMode=BLEND_Translucent, pixelDepthOffsetMode=PDOM_Legacy, shadingModel=MSM_Unlit, usedShadingModels=Unlit}
 nodes [
  n1 = ParticleColor()
  MaterialFunctionCall(Emissive=@n1.RGB) {materialFunction=&MF_FX_EmissiveAdjustments}
  n2 = DynamicParameter() {defaultValue=<rgba 0 0 0 1>, paramNames [Random,1,2,3]}
  n3 = TextureSample() {samplerType=SAMPLERTYPE_LinearColor, texture=&/T_Dive_ExoSuit_Bits}
  n4 = Constant() {r=0.333}
  Multiply(
   A=@n1.A
   B=If(
    A=@n2.Random
    B=Constant() {r=0.666}
    "A > B"=@n3.B
    "A == B"=@n3.B
    "A < B"=If(A=@n2.Random, B=@n4, "A > B"=@n3.R, "A == B"=@n3.R, "A < B"=@n3.R)
   )
  )
 ]
}
