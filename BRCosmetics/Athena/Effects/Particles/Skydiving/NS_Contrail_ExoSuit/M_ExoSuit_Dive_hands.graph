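# Unlit, translucent, two-sided Niagara mesh-particle material for the hands effect of the ExoSuit skydiving dive contrail.
# Particle color RGB passes through MF_FX_EmissiveAdjustments to shape the glow.
# The hands root sprite texture is masked by a Channel vector parameter (red by default), summed and multiplied by particle alpha.
# The channel sum adds red twice plus blue and omits green, so a green selection yields zero.
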Material /BRCosmetics/Athena/Effects/Particles/Skydiving/NS_Contrail_ExoSuit/M_ExoSuit_Dive_hands
graph {
 properties {
  bUsedWithNiagaraMeshParticles=true
  blendMode=BLEND_Translucent
  pixelDepthOffsetMode=PDOM_Legacy
  previewMesh=&/SM_FX_ExoSuit_Dive_hands
  shadingModel=MSM_Unlit
  twoSided=true
  usedShadingModels=Unlit
 }
 nodes [
  n1 = ParticleColor()
  MaterialFunctionCall(Emissive=@n1.RGB) {materialFunction=&MF_FX_EmissiveAdjustments}
  n2 = TextureSample() {
   samplerType=SAMPLERTYPE_LinearColor
   texture=&/T_Dive_ExoSuit_HandsRootSprite
  }
  n3 = MaterialFunctionCall(Float3=Multiply(A=@n2.RGB, B=VectorParameter() {parameterName=Channel, defaultValue=<rgba 1 0 0 0>}.RGB)) {materialFunction=&BreakOutFloat3Components}
  Multiply(A=@n1.A, B=Add(A=Add(A=@n3.R, B=@n3.R), B=@n3.B))
 ]
}
