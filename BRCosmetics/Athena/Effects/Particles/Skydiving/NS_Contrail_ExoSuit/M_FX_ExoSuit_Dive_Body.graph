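# Translucent, unlit, two-sided material for Niagara mesh particles, named as the body effect of the ExoSuit skydiving dive contrail.
# The R, G and B channels of one body texture act as separate masks, each weighted by a dynamic parameter (R, G and Glow opacity) and summed.
# That sum scales particle color, which passes through an emissive adjustment function, and also scales particle alpha.
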Material /BRCosmetics/Athena/Effects/Particles/Skydiving/NS_Contrail_ExoSuit/M_FX_ExoSuit_Dive_Body
graph {
 properties {
  bCastRayTracedShadows=false
  bUsedWithNiagaraMeshParticles=true
  blendMode=BLEND_Translucent
  pixelDepthOffsetMode=PDOM_Legacy
  shadingModel=MSM_Unlit
  twoSided=true
  usedShadingModels=Unlit
 }
 nodes [
  n1 = ParticleColor()
  n2 = TextureSample() {samplerType=SAMPLERTYPE_LinearColor, texture=&/T_Dive_ExoSuit_Body}
  n3 = DynamicParameter() {defaultValue=<rgba 1 0.5 1 1>, paramNames [R Opacity, G Opacity, Glow Opacity, Param4]}
  n4 = Add(A=Multiply(A=@n2.R, B=@n3.R Opacity), B=Add(A=Multiply(A=@n2.G, B=@n3.G Opacity), B=Multiply(A=@n2.B, B=@n3.Glow Opacity)))
  MaterialFunctionCall(Emissive=Multiply(A=@n1.RGB, B=@n4)) {materialFunction=&MF_FX_EmissiveAdjustments}
  Multiply(A=@n1.A, B=@n4)
 ]
}
