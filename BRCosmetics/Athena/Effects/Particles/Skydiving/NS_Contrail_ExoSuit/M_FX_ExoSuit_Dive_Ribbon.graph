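# Translucent, unlit, two-sided Niagara material for the ExoSuit skydiving dive contrail ribbon.
# Dynamic parameters scale, offset and scroll a ribbon texture along U, and vertex color red switches between its red and green channels.
# The sample tints particle-colored emissive and is eroded against a squared falloff around a point along U to shape alpha.
# Particle alpha and a flattened camera-vector-to-normal dot product further fade it.
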Material /BRCosmetics/Athena/Effects/Particles/Skydiving/NS_Contrail_ExoSuit/M_FX_ExoSuit_Dive_Ribbon
graph {
 properties {
  bUsedWithNiagaraMeshParticles=true
  bUsedWithNiagaraRibbons=true
  blendMode=BLEND_Translucent
  pixelDepthOffsetMode=PDOM_Legacy
  shadingModel=MSM_Unlit
  twoSided=true
  usedShadingModels=Unlit
 }
 nodes [
  n1 = ParticleColor()
  n2 = TextureCoordinate()
  n3 = ComponentMask(None=@n2) {r=true}
  n4 = DynamicParameter() {defaultValue=<rgba 0.1 0 1 0>, paramNames [V Scroll Speed, Offset, U_Scale, SpeedChange]}
  n5 = AppendVector(A=Add(A=OneMinus(None=@n3), B=@n4.Offset), B=ComponentMask(None=@n2) {g=true})
  n6 = AppendVector(A=@n4.U_Scale, B=Constant() {r=1})
  n7 = Add(A=@n4.SpeedChange, B=Multiply(A=@n4.V Scroll Speed, B=Time()))
  n8 = Constant()
  n9 = TextureSample(UVs=Add(A=Multiply(A=@n5, B=@n6), B=AppendVector(A=@n7, B=@n8))) {samplerType=SAMPLERTYPE_LinearColor, texture=&/T_Dive_ExoSuit_Ribbon}
  n10 = MaterialFunctionCall(X=@n9.R, Y=@n9.R) {materialFunction=&MakeFloat2}
  n11 = MaterialFunctionCall(Float2=If(
    A=VertexColor().R
    B=Constant() {r=0.5}
    "A > B"=@n10.Result
    "A == B"=@n10.Result
    "A < B"=MaterialFunctionCall(X=@n9.G, Y=@n9.G) {materialFunction=&MakeFloat2}.Result
   )) {materialFunction=&BreakOutFloat2Components}
  MaterialFunctionCall(Emissive=Multiply(A=@n1.RGB, B=@n11.G)) {materialFunction=&MF_FX_EmissiveAdjustments}
  n12 = Abs(None=Subtract(A=@n3) {constB=0.3})
  n13 = Multiply(A=Multiply(A=@n12, B=@n12)) {constB=5}
  n14 = OneMinus(None=Saturate(None=@n13))
  Multiply(A=Multiply(A=@n14, B=@n14), B=@n14)
  n15 = Multiply(A=Subtract(A=@n11.R, B=@n13)) {constB=20}
  n16 = Multiply(A=@n3, B=Multiply(A=@n14, B=Saturate(None=@n15)))
  n17 = MaterialFunctionCall(Float3=CameraVectorWS()) {materialFunction=&BreakOutFloat3Components}
  n18 = MaterialFunctionCall(X=@n17.R, Y=@n17.R, Z=Constant()) {materialFunction=&MakeFloat3}
  n19 = VertexNormalWS()
  Multiply(A=Multiply(A=@n1.A, B=Multiply(A=@n11.G, B=@n16)), B=Abs(None=DotProduct(A=@n18.Result, B=@n19)))
 ]
}
