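# Niagara skydiving dive contrail for the ExoSuit cosmetic: a cyan energy effect attached to the player's skeletal mesh.
# Bone-attached emitters draw a central mesh with light, mirrored spinning hand sprites and meshes, hand glows and drifting bits.
# Two ribbons trail behind, aura sprites travel via bone flow, and animated wave sprites spawn in a ring around the owner.
# System Dive and integrator values blend offsets, spawn rates, rotation and opacity, with some front-end overrides.
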NiagaraSystem /BRCosmetics/Athena/Effects/Particles/Skydiving/NS_Contrail_ExoSuit/NS_Contrail_Dive_ExoSuit
dependencies {
 usedDataInterfaces [
  &NiagaraDataInterfaceVectorCurve
  &NiagaraDataInterfaceCurve
  &NiagaraDataInterfaceArrayColor
  &NiagaraDataInterfaceColorCurve
  &NiagaraDataInterfaceArrayFloat
  &NiagaraDataInterfaceMeshRendererInfo
 ]
 usedDynamicInputs [
  &Divide_Float
  &MakeFloatFromInt
  &ReturnParticleIDAttributes
  &TransformVector
  &SetBoolByFloatComparison
  &Subtract_Float
 ]
 usedModules [
  &SystemState
  /Game/Effects/Niagara/Modules/Skydiving/FN_DivingAttributes/v5/FN_DivingAttributes
  &self:SystemSpawnScript.SystemScriptSource.SystemScriptGraph.NiagaraNodeAssignment_0.SetVariables_#1
  &self:SystemSpawnScript.SystemScriptSource.SystemScriptGraph.NiagaraNodeAssignment_1.SetVariables_#2
  &FN_Integrator
  &self:SystemSpawnScript.SystemScriptSource.SystemScriptGraph.NiagaraNodeAssignment_2.SetVariables_#3
  &self:SystemSpawnScript.SystemScriptSource.SystemScriptGraph.NiagaraNodeAssignment_5.SetVariables_#4
 ]
}
emitters {
 Aura_Hands {
  emitterUpdateScript [
   &EmitterState(Life Cycle Mode=::NewEnumerator1, Loop Duration=1, Recalculate Duration Each Loop=true, UseLoopDelay=false)
   &SpawnRate(SpawnRate=10, Use Spawn Probability=false)
  ]
  particleSpawnScript [
   &FN_SetSpawnGroup(Number of Spawn Groups=2)
   &InitializeParticle(
    Color=&LerpLinearColors(StartColor=<rgba 0.0878 1.9071 2 0.5>, EndColor=<rgba 0.3073 6.6749 7 0.5>, LerpFactor=$System.FN_Integrator001.Value)
    Lifetime=&Lerp_Float(A=1.8, B=1.3, Alpha=$System.FN_Integrator001.Value)
    Sprite Rotation Mode=::NewEnumerator2
    Sprite Size Mode=::NewEnumerator1
    Sprite UV Mode=::NewEnumerator3
    Uniform Sprite Size Max=&Lerp_Float(A=35, B=50, Alpha=$System.FN_Integrator001.Value)
    Uniform Sprite Size Min=25
    UsePositionOffset=false
    Write Material Random=true
   )
   &CameraOffset(Camera Offset Amount=15)
  ]
  particleUpdateScript [
   &ParticleState()
   &DynamicMaterialParameters(Write Parameter Index 0=false)
   /Game/Effects/Niagara/Modules/Sprite/FN_SpriteUV(
    Randomize Sprite UV scale=true
    SubUV Lookup Index:NiagaraFloat=&FloatFromCurve(
     FloatCurve=DataInterface() {propertyValues {Curve {keys [<at 0 0>,<at 1 1>]}}}
     Scale Curve=&MakeCustomFloatFromBool(Bool=&RandomBool(), True Float=1, False Float=-1)
    )
   )
   &ScaleSpriteSize(Scale Factor=&Vector2DFromFloat(Value=&FloatFromCurve(
      FloatCurve=DataInterface() {propertyValues {Curve {keys [
          {arriveTangent=3.4454, interpMode=RCIM_Cubic, leaveTangent=3.4454, tangentMode=RCTM_User, time=0.4, value=0}
          <at 0.5 0.55>
          {interpMode=RCIM_Cubic, tangentMode=RCTM_User, time=0.9, value=1}
          {arriveTangent=-11.0136, interpMode=RCIM_Cubic, leaveTangent=-11.0136, tangentMode=RCTM_User, time=1, value=0.5}
         ]}}}
      Scale Curve=&SelectFloatFromArray(Direct Array Index=$Particles.SpawnGroup, Float Selection Array=DataInterface() {propertyValues {FloatData [1,1,1.25,1.25,2,1.5]}})
     )))
   &ScaleColor(
    Scale Alpha=&FloatFromCurve(FloatCurve=DataInterface() {propertyValues {Curve {keys [
         {interpMode=RCIM_Cubic, tangentMode=RCTM_User, time=0.4, value=0}
         <at 0.5 0.85>
         {interpMode=RCIM_Cubic, tangentMode=RCTM_User, time=0.9, value=1}
         {tangentMode=RCTM_User, time=1, value=0}
        ]}}})
    ScaleA=true
    ScaleRGB=true
   )
   &SolveForcesAndVelocity(Clamp Velocity=false, Limit Acceleration=false)
   &FN_BoneFlow(
    Angle:NiagaraFloat=&RandomRangeFloat(Minimum=-1, Maximum=1)
    Height:NiagaraFloat=&FloatFromCurve(FloatCurve=DataInterface() {propertyValues {Curve {keys [{arriveTangent=1.4394, interpMode=RCIM_Cubic, time=0, value=0}, {interpMode=RCIM_Cubic, tangentMode=RCTM_User, time=1, value=1.05}]}}})
    Interpolate Mesh Sampling=false
    Radius=1
    Set Material Dynamic=false
    SkeletalMesh:NiagaraDataInterfaceSkeletalMesh=$User.SkeletalMesh_Flow
   )
  ]
  rendererClasses [&NiagaraSpriteRendererProperties]
  renderers [{rendererClass=&NiagaraSpriteRendererProperties}]
 }
 Bits {
  emitterUpdateScript [
   &EmitterState()
   &SpawnRate(
    SpawnRate=&Lerp_Float(A=20, B=&FN_FrontEndFloatModifier(Gameplay=50, Front End=45), Alpha=$System.Dive)
    Use Spawn Probability=false
   )
  ]
  particleSpawnScript [
   &InitializeParticle(
    Color=&SelectLinearColorFromArray(
     Array Sampling Mode=::NewEnumerator0
     Direct Array Index=?
     Color Selection Array=DataInterface() {propertyValues {ColorData [<rgba 0.1868 0.7816 1 1>, <rgba 0.0591 0.3784 0.401 0.5>]}}
     Random Seed=0
    )
    Lifetime Max=1.5
    Lifetime Mode=::NewEnumerator1
    Mesh Scale Mode=::NewEnumerator3
    Mesh Uniform Scale=&RandomRangeFloat(Minimum=1.2, Maximum=1.45)
    Sprite Rotation Mode=::NewEnumerator2
    Sprite Size Mode=::NewEnumerator1
    Uniform Sprite Size Max=15
    UsePositionOffset=false
    Write Material Random=true
   )
   &SkeletalMeshLocation(
    Bone / Socket Index=&RandomRangeInt(Minimum=0, Maximum=1)
    Bone Sampling Mode=::NewEnumerator4
    MaskSpawnGroup=false
    Sampled Position Offset=&Lerp_Vector(A=<0 0 21>, B=<0 0 21>, Alpha=$System.Dive)
    Sampled Position Offset Coordinate Space:ENiagaraSkelMeshTransforms=::NewEnumerator2
    Skeletal Mesh=$User.SkeletalMesh
   )
   &BoxLocation(Box Offset=<0 0 30>, Box Size=<0 0 60>, MaskSpawnGroup=false, Module.Override Seed=false, Rotate Box=false, Surface Only=false)
   &AddVelocity(Velocity=&Lerp_Vector(
     A=&RandomRangeVector(Minimum=<0 0 50>, Maximum=<0 0 100>)
     B=&RandomRangeVector(Minimum=<0 0 100>, Maximum=<0 0 200>)
     Alpha=$System.Dive
    ))
   &FN_SetSpawnGroup()
   &DynamicMaterialParameters(
    Index 0 Param 1=HlslExpression() {hlslExpression="rand(1)"}
    Module.Param0WriteEnabled=true
    Module.Param1WriteEnabled=false
    Module.Param2WriteEnabled=false
    Module.Param3WriteEnabled=false
   )
  ]
  particleUpdateScript [
   &ParticleState()
   /Game/Effects/Niagara/Modules/Skydiving/FN_DivingAttributes/FN_DivingAttributes_v2(
    Accumulate Dive=1
    Accumulate Rest=0
    Bone Index=0
    SkeletalMesh=DataInterface() {propertyValues {
      DefaultMesh=null
      ExcludeBoneName=root
      OptionalComponentClass=null
      PreviewMesh=&M_Med_Soldier_04_Preview
      SkinningMode=SkinOnTheFly
      SoftSourceActor=null
      SourceMode=Default
      UvSetIndex=0
      WholeMeshLOD=-1
      bExcludeBone=false
      bReadDeformedGeometry=true
      bRequireCurrentFrameData=true
      ComponentTags []
      FilteredBones [spine_02]
      FilteredSockets []
      MeshUserParameter {parameter {name=None, typeDefHandle {registeredTypeIndex=1}, varData []}}
      SamplingRegions []
     }}
   )
   &InheritVelocity(Inherited Velocity Amount Scale=<1 1 0>, Inherited Velocity Speed Limit=0, Speed Limit=true, UseSpeedThreshold=false)
   &CurlNoiseForce(Noise Frequency=5, Noise Quality / Cost=::NewEnumerator0, Noise Strength=500, Pan Noise=true, Pan Noise Field=<15 10 10>)
   &SolveForcesAndVelocity(Clamp Velocity=false, Limit Acceleration=false)
   &SpriteRotationRate(Rotation Rate=100)
   &ScaleSpriteSize(Uniform Curve Sprite Scale=DataInterface() {propertyValues {Curve {keys [
        {arriveTangent=6.8904, interpMode=RCIM_Cubic, leaveTangent=6.8904, tangentMode=RCTM_User, time=0, value=0}
        {interpMode=RCIM_Cubic, tangentMode=RCTM_User, time=0.5, value=1}
        {tangentMode=RCTM_User, time=1, value=0}
       ]}}})
  ]
  rendererClasses [&NiagaraSpriteRendererProperties]
  renderers [{rendererClass=&NiagaraSpriteRendererProperties}]
 }
 Center {
  emitterSpawnScript [&FN_PreserveSkydiveUserVariables() {module=PreserveSkydiveUserVariables}]
  emitterUpdateScript [
   &EmitterState(
    Life Cycle Mode=::NewEnumerator1
    Loop Behavior=::NewEnumerator1
    Loop Duration Mode=::NewEnumerator1
    Reset Age On Awaken=true
    Scalability Mode=::NewEnumerator1
    UseLoopDelay=false
   )
   &SpawnBurst_Instantaneous(Spawn Count=1, Use Spawn Probability=false)
  ]
  particleSpawnScript [
   &FN_SetSpawnGroup(Number of Spawn Groups=1) {module=FN_SetSpawnGroup001}
   /Niagara/Modules/Spawn/Initialization/InitializeParticle(
    Color=<rgba 0.1317 2.8607 3 0.15>
    Lifetime=2
    Mesh Scale=<1.7 1.7 2.2>
    Module.Write Color=true
    Module.Write Lifetime=true
    Module.Write Mass=false
    Module.Write Material Random=true
    Module.Write Position=false
    Module.Write Scale=true
    Module.Write SpriteRotation=false
    Module.Write SpriteSize=false
    Position=&SimulationPosition()
    Sprite Rotation=&UniformRangedFloat(Minimum=0, Maximum=1)
   )
   &SkeletalMeshLocation(
    Bone / Socket Index=5
    Bone Sampling Mode=::NewEnumerator4
    Inherit Velocity=false
    Interpolate Mesh Sampling=false
    MaskSpawnGroup=false
    Sampled Position Offset=<0 0 20>
    Sampled Position Offset Coordinate Space:ENiagaraMeshTransforms=::NewEnumerator3
    Skeletal Mesh=$User.SkeletalMesh
   )
   &InitialMeshOrientation(
    Orientation Axis=<0 0 1>
    Orientation Coordinate Space=::Local
    Orientation Vector=<0 0 1>
    Use Orientation Vector=true
    Use Rotation Vector=false
   )
  ]
  particleUpdateScript [
   &FN_DivingAttributes_v2(Accumulate Dive=0.1, Accumulate Rest=0.1, SkeletalMesh:NiagaraDataInterfaceSkeletalMesh=$User.SkeletalMesh) {module=FN_DivingAttributes_v2001}
   &ParticleState(Kill Particles When Lifetime Has Elapsed=false, Let Infinitely Lived Particles Die When Emitter Deactivates=true)
   &SolveForcesAndVelocity(Clamp Velocity=false, Limit Acceleration=false)
   &ScaleMeshSize(Scale Factor=&VectorFromCurve(
     VectorCurve=DataInterface() {propertyValues {
       XCurve {keys [{arriveTangent=1, interpMode=RCIM_Cubic, leaveTangent=1, time=0, value=0}, {tangentMode=RCTM_User, time=0.9, value=1}]}
       YCurve {keys [{arriveTangent=1, interpMode=RCIM_Cubic, leaveTangent=1, time=0, value=0}, {tangentMode=RCTM_User, time=0.9, value=1}]}
       ZCurve {keys [{tangentMode=RCTM_User, time=0, value=1.1}, {tangentMode=RCTM_User, time=1, value=1}]}
      }}
     CurveIndex=$Emitter.Age
    ))
   &ScaleColor(
    Scale Alpha=&FloatFromCurve(
     FloatCurve=DataInterface() {propertyValues {Curve {keys [
         {arriveTangent=4.0516, interpMode=RCIM_Cubic, leaveTangent=4.0516, tangentMode=RCTM_User, time=0, value=0}
         {arriveTangent=0.0961, interpMode=RCIM_Cubic, leaveTangent=0.0961, tangentMode=RCTM_User, time=0.25, value=1}
         {arriveTangent=0.2272, interpMode=RCIM_Cubic, leaveTangent=0.2272, tangentMode=RCTM_User, time=1, value=1}
        ]}}}
     Scale Curve=&Lerp_Float(A=0.2, B=0.5, Alpha=$System.Dive)
    )
    ScaleA=true
    ScaleRGB=false
   )
   &UpdateMeshOrientation(
    Rotation Coordinate Space=::NewEnumerator2
    Rotation Rate=&Lerp_Float(A=0.09, B=0.3, Alpha=$System.FN_Integrator001.Value)
    Rotation Vector=<0 0 1>
   )
   &DynamicMaterialParameters(
    Index 0 Param 1=1
    Index 0 Param 2=0.25
    Index 0 Param 3=4
    Module.Param0WriteEnabled=true
    Module.Param1WriteEnabled=true
    Module.Param2WriteEnabled=true
    Module.Param3WriteEnabled=true
   )
   &FN_Light_Attributes(
    Color=<rgba 0.3951 8.582 9 1>
    Create Color Attribute=true
    Create Exponent Attribute=true
    Create Radius Attribute=true
    Exponent=5
    Radius=150
   )
  ]
  rendererClasses [&NiagaraMeshRendererProperties, &NiagaraLightRendererProperties]
  renderers [
   {rendererClass=&NiagaraMeshRendererProperties, rendererIndex=0}
   {rendererClass=&NiagaraLightRendererProperties, rendererIndex=1}
  ]
 }
 FN_Contrails_Ribbon_v4 {
  emitterUpdateScript [
   &EmitterScalability(
    Module.bApplySpawnCountScale=false
    Module.bApplySpawnCountScaleByDistanceFraction=false
    Module.bUseMaxDistance=false
    Module.bUseMinDistance=false
   ) {enabled=false}
   &EmitterLifeCycle(CompleteOnInactive=true) {enabled=false}
   &EmitterState()
   &FN_SpawnRibbon_v2(Number of Ribbons=2, Spawn Mode=::NewEnumerator0)
  ]
  particleSpawnScript [
   &FN_InitializeRibbon_v3(Enable Skeletal Position=true, Lifetime=0.7, Ribbon Width=60) {enabled=false}
   &FN_InitializeRibbon_v3(
    Color:LinearColor=/Niagara/DynamicInputs/TypeConversions/MakeLinearColorFromFloat(Value:NiagaraFloat=0.6)
    Enable Skeletal Position=true
    Lifetime=0.7
    Ribbon Width=50
    Sampled Position Coordinate Space:ENiagaraCoordinateSpace=::Local
    Sampled Rotation Coordinate Space:ENiagaraCoordinateSpace=::Local
    Sampled Velocity Coordinate Space:ENiagaraCoordinateSpace=::Local
   ) {module=FN_InitializeRibbon_v3001}
   &AddVelocity(Velocity=<0 0 250>)
   &AddVelocity(Velocity=&MakeCustomVectorFromBool(
     Bool=$Particles.Locked
     True Vector=<0 0 0>
     False Vector=&FN_FrontEndVector3Modifier(Gameplay=<0 0 100>, Front End=<0 0 100>)
    )) {module=AddVelocity001}
   &SampleSkeletalMeshSkeleton(Mesh:NiagaraDataInterfaceSkeletalMesh=$User.SkeletalMesh)
  ]
  particleUpdateScript [
   /Game/Effects/Niagara/Modules/Skydiving/FN_DivingAttributes/FN_DivingAttributes() {enabled=false}
   &SolveForcesAndVelocity(Clamp Velocity=false, Limit Acceleration=false)
   &FN_UpdateRibbon(
    Enable Skeletal Position=true
    Initial Color:LinearColor=$Particles.Initial.Color
    Initial RibbonWidth:NiagaraFloat=$Particles.Initial.RibbonWidth
   )
   &ScaleColor(
    Linear Color Curve=DataInterface() {propertyValues {
      AlphaCurve {keys [<at 0 0.4>,<at 0.7316 0.4>,<at 1 0.1>]}
      BlueCurve {keys [
        <at 0.0013 0.6962>
        {arriveTangent=-0.7216, interpMode=RCIM_Cubic, leaveTangent=-0.7216, time=0.3545, value=1}
        {arriveTangent=-0.7216, interpMode=RCIM_Cubic, leaveTangent=-0.7216, time=0.9661, value=0}
       ]}
      GreenCurve {keys [
        <at 0.0013 0.6216>
        {arriveTangent=-0.6443, interpMode=RCIM_Cubic, leaveTangent=-0.6443, time=0.3545, value=0.9166}
        {arriveTangent=-0.6443, interpMode=RCIM_Cubic, leaveTangent=-0.6443, time=0.9661, value=0}
       ]}
      RedCurve {keys [
        <at 0.0013 0>
        <at 0.3545 0.2552 RCIM_Cubic>
        {arriveTangent=-0.3741, interpMode=RCIM_Cubic, leaveTangent=-0.3741, tangentMode=RCTM_User, time=0.9661, value=0}
       ]}
     }}
    Scale Mode=::NewEnumerator2
   ) {module=ScaleColor001}
   &Material(Base Texture Time:NiagaraFloat=&Multiply_Float(A=$Emitter.Age, B=-1), PostMult=0.5455, Power=0.4408)
   &ScaleColor(
    Scale Alpha=&FloatFromCurve(
     FloatCurve=DataInterface() {propertyValues {Curve {keys [{tangentMode=RCTM_User, time=0.56, value=0}, {tangentMode=RCTM_User, time=1, value=1}]}}}
     CurveIndex=$Emitter.Age
    )
    ScaleA=true
    ScaleRGB=true
   )
  ]
  rendererClasses [&NiagaraRibbonRendererProperties]
  renderers [{rendererClass=&NiagaraRibbonRendererProperties}]
 }
 Glow_Hands {
  emitterUpdateScript [
   &EmitterState(Life Cycle Mode=::NewEnumerator1, Loop Duration=1, UseLoopDelay=false)
   &SpawnRate(SpawnRate=&Lerp_Float(A=6, B=12, Alpha=$System.FN_Integrator001.Value), Use Spawn Probability=true)
   &SpawnBurst_Instantaneous(Spawn Count=2, Use Spawn Probability=false, UseLoopCountLimit=true)
  ]
  particleSpawnScript [
   &FN_SetSpawnGroup(Number of Spawn Groups=2)
   /Niagara/Modules/Spawn/Initialization/V2/InitializeParticle(
    Color=<rgba 0 0.6387 2 0.333>
    Lifetime=&Lerp_Float(A=2, B=1.25, Alpha=$System.FN_Integrator001.Value)
    Sprite Rotation Mode=::NewEnumerator2
    Sprite Size Mode=::NewEnumerator1
    Uniform Sprite Size Max=180
    Uniform Sprite Size Min=150
    UsePositionOffset=false
    Write Material Random=true
   )
  ]
  particleUpdateScript [
   &SolveForcesAndVelocity(Clamp Velocity=false, Limit Acceleration=false)
   &ParticleState()
   &SkeletalMeshLocation(
    Bone / Socket Index=$Particles.SpawnGroup
    Bone Sampling Mode=::NewEnumerator4
    Inherit Velocity=false
    Interpolate Mesh Sampling=false
    MaskSpawnGroup=false
    Sampled Position Offset=&Multiply_Vector(
     A=&Lerp_Vector(A=<0 11 12>, B=<0 -7 12>, Alpha=$System.Dive)
     B=&Lerp_Vector(A=<1 1 1>, B=<1 1 3>, Alpha=$Particles.NormalizedAge)
    )
    Skeletal Mesh=$User.SkeletalMesh
   )
   &ScaleSpriteSize(Scale Factor=&Vector2DFromFloat(Value=&FloatFromCurve(FloatCurve=DataInterface() {propertyValues {Curve {keys [{arriveTangent=1.5832, time=0, value=0.7}, <at 0.1 1>, {leaveTangent=0.5893, time=1, value=0.5}]}}})))
   &ScaleColor(
    Scale Alpha=&FloatFromCurve(
     FloatCurve=DataInterface() {propertyValues {Curve {keys [<at 0 0 RCIM_Cubic>,<at 0.15 1>,<at 0.6 1 RCIM_Cubic>,<at 1 0 RCIM_Cubic>]}}}
     Scale Curve=0.15
    )
    Scale RGB=<0.5 0.5 0.5>
    ScaleA=true
    ScaleRGB=true
   )
   &CameraOffset(Camera Offset Amount=-15)
   &DynamicMaterialParameters(
    Index 0 Param 2=&FloatFromCurve(FloatCurve=DataInterface() {propertyValues {Curve {keys [
         {interpMode=RCIM_Cubic, tangentMode=RCTM_User, time=0.4, value=0.1}
         {arriveTangent=1.4285, interpMode=RCIM_Cubic, leaveTangent=1.4285, tangentMode=RCTM_User, time=1, value=1}
        ]}}})
    Module.Param0WriteEnabled=true
    Module.Param1WriteEnabled=true
    Module.Param2WriteEnabled=false
    Module.Param3WriteEnabled=false
   )
  ]
  rendererClasses [&NiagaraSpriteRendererProperties]
  renderers [{rendererClass=&NiagaraSpriteRendererProperties}]
 }
 Hands {
  emitterUpdateScript [
   &EmitterState(
    Inactive Response=::NewEnumerator1
    Life Cycle Mode=::NewEnumerator1
    Loop Behavior=::NewEnumerator1
    Loop Duration Mode=::NewEnumerator1
    UseLoopDelay=false
   ) {module=EmitterState001}
   &SpawnBurst_Instantaneous(Use Spawn Probability=false, UseLoopCountLimit=false)
   &SpawnBurst_Instantaneous(Spawn Time=0.0607, Use Spawn Probability=false, UseLoopCountLimit=false) {module=SpawnBurst_Instantaneous001}
  ]
  particleSpawnScript [
   &FN_SetSpawnGroup(Number of Spawn Groups=2) {module=FN_SetSpawnGroup001}
   &self:FlipSpliteAnotherHand() {module=FlipSpliteAnotherHand}
   &FN_SetSpawnGroup(Number of Spawn Groups=2)
   /Niagara/Modules/Spawn/Initialization/InitializeParticle(
    Color=<rgba 0.1317 2.8607 3 0.5>
    Lifetime=&UniformRangedFloat(Minimum=0.9, Maximum=1.1)
    Mesh Scale=<4 4 4>
    Module.Write Color=true
    Module.Write Lifetime=false
    Module.Write Mass=false
    Module.Write Material Random=true
    Module.Write Position=false
    Module.Write Scale=true
    Module.Write SpriteRotation=false
    Module.Write SpriteSize=true
    Position=&SimulationPosition()
    Sprite Rotation=&UniformRangedFloat(Minimum=0, Maximum=360)
    Sprite Size=&Vector2DFromFloat(Value=56)
   )
   /SetVariables(Particles.InitTimeOfs:NiagaraFloat=&Multiply_FloatByInt(Float=0.5, Integer=$Particles.UniqueID)) {isSetParameters=true, module=SetVariables_#5}
   &InitialMeshOrientation(Mesh Orientation Mode=::NewEnumerator7, Rotation=<0 -0.1 0>, Use Rotation Vector=true)
   /SetVariables(Particles.UVScale:Vector2f=&Multiply_Vector2D(A=<1 1>, B=HlslExpression() {hlslExpression="float2(Particles.FlipOtherHand, 1.0) /* Custom HLSL! */"})) {isSetParameters=true}
  ]
  particleUpdateScript [
   &ParticleState() {enabled=false}
   &SkeletalMeshLocation(
    Bone / Socket Index=$Particles.SpawnGroup
    Bone Sampling Mode=::NewEnumerator4
    Inherit Velocity=false
    "Inherit Velocity (Scale)"=0
    MaskSpawnGroup=false
    Sampled Position Offset=&FN_FrontEndVector3Modifier(
     Gameplay=&Lerp_Vector(A=<0 15 10>, B=<0 -7 2>, Alpha=$System.Dive)
     Front End=&Lerp_Vector(A=<0 11 5>, B=<0 -7 5>, Alpha=$System.Dive)
    )
    Sampled Position Offset Coordinate Space:ENiagaraMeshTransforms=::NewEnumerator2
    Skeletal Mesh=$User.SkeletalMesh
   )
   &DynamicMaterialParameters(Module.Param0WriteEnabled=true, Module.Param1WriteEnabled=false, Module.Param2WriteEnabled=false, Module.Param3WriteEnabled=false)
   &CameraOffset(Camera Offset Amount=25)
   &ScaleColor(
    Scale Alpha=&Lerp_Float(
     A=0.9
     B=&Sine(Normalized Angle=$Emitter.Age, Period=0.5, Scale=0.075, Bias=1)
     Alpha=$System.FN_Integrator001.Value
    )
    ScaleA=true
    ScaleRGB=true
   )
   &UpdateMeshOrientation(Rotation Rate=&Multiply_Float(A=0.5, B=$Particles.FlipOtherHand), Rotation Vector=<0 0 1>)
  ]
  rendererClasses [&NiagaraSpriteRendererProperties, &NiagaraMeshRendererProperties]
  renderers [
   {rendererClass=&NiagaraSpriteRendererProperties}
   {rendererClass=&NiagaraMeshRendererProperties, rendererIndex=1}
  ]
 }
 RibbonMesh {
  emitterUpdateScript [
   &EmitterState(
    Inactive Response=::NewEnumerator1
    Life Cycle Mode=::NewEnumerator1
    Loop Behavior=::NewEnumerator1
    Loop Duration Mode=::NewEnumerator1
    UseLoopDelay=false
   ) {module=EmitterState001}
   &SpawnBurst_Instantaneous(Spawn Count=2, Use Spawn Probability=false, UseLoopCountLimit=false)
   &SpawnBurst_Instantaneous(Spawn Count=2, Spawn Time=0.0607, Use Spawn Probability=false, UseLoopCountLimit=false) {module=SpawnBurst_Instantaneous001}
  ]
  particleSpawnScript [
   &FN_SetSpawnGroup(Number of Spawn Groups=2) {module=FN_SetSpawnGroup001}
   &self:FlipSpliteAnotherHand() {module=FlipSpliteAnotherHand}
   &FN_SetSpawnGroup(Number of Spawn Groups=2)
   &InitializeParticle(
    Color=<rgba 0.1 1.2 1.5 4>
    Lifetime=&UniformRangedFloat(Minimum=0.9, Maximum=1.1)
    Mesh Scale=&Multiply_Vector(A=<0.9 0.9 0.75>, B=HlslExpression() {hlslExpression="float3(Particles.FlipOtherHand, 1.0, 1.0) /* Custom HLSL! */"})
    Module.Write Color=true
    Module.Write Lifetime=false
    Module.Write Mass=false
    Module.Write Material Random=true
    Module.Write Position=false
    Module.Write Scale=true
    Module.Write SpriteRotation=false
    Module.Write SpriteSize=false
    Position=&SimulationPosition()
    Sprite Rotation=&UniformRangedFloat(Minimum=0, Maximum=360)
    Sprite Size=&Vector2DFromFloat(Value=&Multiply_Float(
      A=&RandomRangeFloat(Minimum=45, Maximum=55)
      B=&MakeCustomFloatFromBool(Bool=&RandomBool(), True Float=-1, False Float=1)
     ))
   )
   &InitialMeshOrientation(Mesh Orientation Mode=::NewEnumerator5, Rotation=<0 0 0.25>, Rotation Coordinate Space=::NewEnumerator2, Use Rotation Vector=true)
   /SetVariables(Particles.InitTimeOfs:NiagaraFloat=&Multiply_FloatByInt(Float=0.5, Integer=$Particles.UniqueID)) {isSetParameters=true, module=SetVariables_#5}
   /SetVariables(Particles.InitRandom=HlslExpression() {hlslExpression="rand(1.0)"}) {isSetParameters=true}
  ]
  particleUpdateScript [
   &ParticleState() {enabled=false}
   &SkeletalMeshLocation(
    Bone / Socket Index=$Particles.SpawnGroup
    Bone Sampling Mode=::NewEnumerator4
    Inherit Velocity=false
    "Inherit Velocity (Scale)"=0
    MaskSpawnGroup=false
    Sampled Position Offset=&Lerp_Vector(A=<0 11 3>, B=<0 -7 3>, Alpha=$System.Dive)
    Sampled Position Offset Coordinate Space:ENiagaraMeshTransforms=::NewEnumerator2
    Skeletal Mesh=$User.SkeletalMesh
   )
   &DynamicMaterialParameters(
    Index 0 Param 1=-0.3
    Index 0 Param 2=$Particles.InitRandom
    Index 0 Param 3=1.5
    Index 0 Param 4=&Add_Float(A=HlslExpression() {hlslExpression=Particles.DynamicMaterialParameter.w}, B=-0.02)
    Module.Param0WriteEnabled=true
    Module.Param1WriteEnabled=true
    Module.Param2WriteEnabled=true
    Module.Param3WriteEnabled=true
   ) {module=DynamicMaterialParameters001}
   &CameraOffset(Camera Offset Amount=5)
   &UpdateMeshOrientation(Rotation Rate=0.1, Rotation Vector=<0 0 1>)
  ]
  rendererClasses [&NiagaraMeshRendererProperties]
  renderers [{rendererClass=&NiagaraMeshRendererProperties, rendererIndex=0}]
 }
 Wave_Center {
  emitterUpdateScript [
   &EmitterState()
   &SpawnRate(SpawnRate=&Lerp_Float(A=0, B=60, Alpha=$System.FN_Integrator001.Value), Use Spawn Probability=false)
  ]
  particleSpawnScript [
   /Niagara/Modules/Spawn/Initialization/V2/InitializeParticle(
    Color=<rgba 1 1.1 1 0.1>
    Lifetime=&RandomRangeFloat(Minimum=0.3, Maximum=0.4)
    Sprite Size Max=<10 50>
    Sprite Size Min=<10 30>
    Sprite Size Mode=::NewEnumerator2
    UsePositionOffset=false
    Write Material Random=true
   )
   /SetVariables(Particles.Position=&Add_Vector(A=Linked() {linkedVariable {name=Engine.Owner.Position, type=NiagaraPosition}}, B=<0 0 55>)) {isSetParameters=true}
   &TorusLocation(Align Torus With Axis=false, Handle Radius=0, Large Radius=40, MaskSpawnGroup=false, Module.Override Randomness=false)
   /SetVariables(Particles.RandomFrame=&RandomRangeInt(Minimum=0, Maximum=63)) {isSetParameters=true}
  ]
  particleUpdateScript [
   &ParticleState()
   &InheritSourceMovement(
    Applied Force Scale=<-0.5 -0.5 0>
    ApplyForce=true
    ApplyForceFalloff=false
    ApplyPosition=true
    ApplyPositionFalloff=false
    Module.LimitForceSpeed=false
   ) {module=InheritSourceMovement001}
   &AddVelocityFromPoint(
    Module.Use Falloff Distance=false
    Origin Offset=<0 0 22>
    Velocity Strength=&FloatFromCurve(
     FloatCurve=DataInterface() {propertyValues {Curve {keys [
         {arriveTangent=-1.0543, interpMode=RCIM_Cubic, time=0.25, value=1}
         {arriveTangent=-5, interpMode=RCIM_Cubic, time=0.6, value=0}
         <at 0.7 -1 RCIM_Cubic>
        ]}}}
     Scale Curve=13
    )
   ) {enabled=false}
   &PointForce(
    Force Falloff Distance=155
    Force Origin=&Add_Vector(A=/Niagara/DynamicInputs/Helpers/SimulationPosition(), B=<0 0 50>)
    Force Strength=1500
    Module.Use Falloff=true
    Module.UseLinearForce=false
   )
   &GravityForce(Gravity=&MakeVector(
     X=0
     Y=0
     Z=&FloatFromCurve(
      FloatCurve=DataInterface() {propertyValues {Curve {keys [<at 0.4 0 RCIM_Cubic>, {arriveTangent=0.5303, interpMode=RCIM_Cubic, leaveTangent=0.5303, tangentMode=RCTM_User, time=0.5, value=1}]}}}
      Scale Curve=2500
     )
    ))
   &SolveForcesAndVelocity(Clamp Velocity=false, Limit Acceleration=false)
   &ScaleColor(
    Scale Alpha=&FloatFromCurve(
     FloatCurve=DataInterface() {propertyValues {Curve {keys [
         {arriveTangent=-0.7151, interpMode=RCIM_Cubic, leaveTangent=-0.7151, tangentMode=RCTM_User, time=0.2, value=0}
         {arriveTangent=-2.5, interpMode=RCIM_Cubic, leaveTangent=-2.5, time=0.6, value=1}
         {arriveTangent=-1.6379, interpMode=RCIM_Cubic, leaveTangent=-1.6379, tangentMode=RCTM_User, time=1, value=0}
        ]}}}
     Scale Curve=0.2
    )
    Scale RGB=&MakeVectorFromLinearColorRGB(Color=&ColorFromCurve(ColorCurve=DataInterface() {propertyValues {
        AlphaCurve {keys [<at 0.3745 0>, {arriveTangent=-5.0984, interpMode=RCIM_Cubic, leaveTangent=-5.0984, time=0.529, value=1}, <at 0.7251 0>]}
        BlueCurve {keys [<at 0.0463 5>]}
        GreenCurve {keys [<at 0.0463 4.1833>]}
        RedCurve {keys [<at 0.0463 0.2018>]}
       }}))
    ScaleA=true
    ScaleRGB=true
   )
   &ScaleSpriteSize(Scale Factor=&Vector2DFromFloat(Value=&FloatFromCurve(
      FloatCurve=DataInterface() {propertyValues {Curve {keys [{arriveTangent=-0.2263, interpMode=RCIM_Cubic, leaveTangent=-0.2263, tangentMode=RCTM_User, time=0, value=0}, <at 1 1 RCIM_Cubic>]}}}
      Scale Curve=3.5
     )))
   &FN_SubUVAnimation(End Frame=&Add_Integer(A=$Particles.RandomFrame, B=24), Start Frame=$Particles.RandomFrame)
  ]
  rendererClasses [&NiagaraSpriteRendererProperties]
  renderers [{rendererClass=&NiagaraSpriteRendererProperties}]
 }
}
summary {
 systemName=NS_Contrail_Dive_ExoSuit
 emitters [
  {emitterName=Center, rendererClasses [&NiagaraMeshRendererProperties, &NiagaraLightRendererProperties]}
  {emitterName=Hands, rendererClasses [&NiagaraSpriteRendererProperties, &NiagaraMeshRendererProperties]}
  {emitterName=RibbonMesh, rendererClasses [&NiagaraMeshRendererProperties]}
  {emitterName=Bits, rendererClasses [&NiagaraSpriteRendererProperties]}
  {emitterName=FN_Contrails_Ribbon_v4, rendererClasses [&NiagaraRibbonRendererProperties]}
  {emitterName=Wave_Center, rendererClasses [&NiagaraSpriteRendererProperties]}
  {emitterName=Aura_Hands, rendererClasses [&NiagaraSpriteRendererProperties]}
  {emitterName=Glow_Hands, rendererClasses [&NiagaraSpriteRendererProperties]}
 ]
 userVariables [
  User.bIsFrontEnd {type=NiagaraBool}
  User.bIsFrontEndPreview {type=NiagaraBool}
  User.ColorChange {type=LinearColor, description="", defaultValue {struct=LinearColor, value=<rgba 1 1 1 1>}}
  User.ParaGlideLeanAlpha {type=NiagaraFloat, defaultValue=0, description=""}
  User.PlayerSpeed {type=NiagaraFloat, defaultValue=0, description=""}
  User.RibbonIDOffset {type=NiagaraInt32}
  User.SkeletalMesh {
   type=NiagaraDataInterfaceSkeletalMesh
   defaultValue {
    struct=NiagaraExt_VariableValue_DataInterface
    value {dataInterface=&self:NiagaraDataInterfaceSkeletalMesh_0}
   }
  }
  User.SkeletalMesh_Flow {
   type=NiagaraDataInterfaceSkeletalMesh
   defaultValue {
    struct=NiagaraExt_VariableValue_DataInterface
    value {dataInterface=&self:NiagaraDataInterfaceSkeletalMesh_3}
   }
  }
 ]
}
